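; Quest gameplay ability that reports an Avacado egg water-removal stat when a player with the AvacadoEaterBird item tag surface swims.
; On activation it prints a debug string, then waits on the owning pawn for that item tag, then for the SurfaceSwimming tag.
; Swimming sends the custom stat event through the quest manager and ends the ability.
; A tag-removal wait also starts, but its Removed output does nothing.
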(event Ability|EventActivateAbility
  (bind _owningplayerpawn (Variables|GAVars|Data|GetOwningPlayerPawn))
  (Development|PrintString "AvacadoOn")
  (|CallParentFunction)
  (Utilities|IsValid _owningplayerpawn
    (:Is%20Valid
      (Ability|Tasks|WaitGameplayTagAddToActor _owningplayerpawn "(TagName=\"Granted.Athena.EnvItem.AvacadoEaterBird\")"
        (:Added
          (Ability|Tasks|WaitGameplayTagRemove "(TagName=\"Granted.Athena.EnvItem.AvacadoEaterBird\")"
            (:then
              _)
            (:Removed
              _))
          (Ability|Tasks|WaitGameplayTagAddToActor _owningplayerpawn "(TagName=\"Gameplay.Action.Player.SurfaceSwimming\")" true
            (:Added
              (Quest|SendComplexCustomStatEvent (Quests|GetQuestManager (Fortnite|GetFortPlayerControllerfromActor _owningplayerpawn) "Invalid") _owningplayerpawn "(GameplayTags=((TagName=\"Athena.Quests.AvacadoEater.Egg.WaterRemoval\")))" "(GameplayTags=)" 1)
              (Ability|EndAbility))))))))

(event Ability|EventFailedToActivatePassiveAbility
  (params (ActorInfo :type Gameplay%20Ability%20Actor%20Info%20Structure)))

(event Ability|EventOnEndAbility
  (params (bWasCancelled :type Boolean)))
