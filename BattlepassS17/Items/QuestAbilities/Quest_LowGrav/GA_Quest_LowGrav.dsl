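; Battle pass Season 17 quest ability that records when the owning player gains low gravity.
; On activation it validates the owning pawn, then waits for the Status.LowGravity gameplay tag to be added to it.
; When the tag arrives, it sends an Athena.Quests.Lowgrav custom stat event to the player's quest manager and ends, so each activation reports once.
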(event Ability|EventActivateAbility
  (bind _owningplayerpawn (Variables|GAVars|Data|GetOwningPlayerPawn))
  (|CallParentFunction)
  (Utilities|IsValid _owningplayerpawn
    (:Is%20Valid
      (Ability|Tasks|WaitGameplayTagAddToActor _owningplayerpawn "(TagName=\"Status.LowGravity\")"
        (:Added
          (Quest|SendComplexCustomStatEvent (Quests|GetQuestManager (Fortnite|GetFortPlayerControllerfromActor _owningplayerpawn) "Invalid") _owningplayerpawn "(GameplayTags=((TagName=\"Athena.Quests.Lowgrav\")))" "(GameplayTags=)" 1)
          (Ability|EndAbility))))))

(event Ability|EventFailedToActivatePassiveAbility
  (params (ActorInfo :type Gameplay%20Ability%20Actor%20Info%20Structure)))

(event Ability|EventOnEndAbility
  (params (bWasCancelled :type Boolean)))
